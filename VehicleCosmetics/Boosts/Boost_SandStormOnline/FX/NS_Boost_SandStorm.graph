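# Niagara sprite effect for the SandStorm vehicle boost, streaming a sand-colored trail and debris behind the car.
# MainTrail and short-lived StartFiller form the plume, while Rocks and Twig add spinning, falling debris.
# Each emitter partitions spawns across filtered skeletal-mesh sockets and inherits most source movement, keeping particles with the car.
# Painted-variant overrides replace default tints, with R1 colors driving boosted spark color; frontend preview alters debris launch velocity.
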NiagaraSystem /VehicleCosmetics/Boosts/Boost_SandStormOnline/FX/NS_Boost_SandStorm
dependencies {
 usedDataInterfaces [
  &NiagaraDataInterfaceCurve
  &NiagaraDataInterfaceSpriteRendererInfo
  &NiagaraDataInterfaceSkeletalMesh
 ]
 usedDynamicInputs []
 usedModules [
  &self:SystemSpawnScript.SystemScriptSource.SystemScriptGraph.NiagaraNodeAssignment_0.SetVariables_#1
  &SystemState
 ]
}
emitters {
 MainTrail_NewTemplate001 {
  emitterUpdateScript [
   &EmitterState()
   &VC_FilteredSocketCount()
   &SpawnRate(
    SpawnRate=&MakeCustomFloatFromBool(Bool=$User.bIsFrontendPreview, True Float=60, False Float=60)
    Use Spawn Probability=false
   )
  ]
  particleSpawnScript [
   &VC_PaintedVariantsOverride(
    Black=<rgba 0.02 0.02 0.02 1.2>
    BurntSienna=<rgba 0.1094 0.0338 0.0135 1>
    Cobalt=<rgba 0.026 0.0441 1 1>
    Crimson=<rgba 0.2656 0 0.0072 1>
    ForestGreen=<rgba 0 0.4167 0.0211 1.2>
    Grey=<rgba 0.1354 0.1354 0.1354 1>
    Lime=<rgba 0.3136 0.724 0 1>
    Orange=<rgba 0.4688 0.1635 0 1>
    Pink=<rgba 0.8557 0.2238 0.8594 1.2>
    Purple=<rgba 0.1261 0 0.3073 1>
    R1 Black:LinearColor=<rgba 0.4219 0.4219 0.4219 0.5>
    R1 BurntSienna:LinearColor=<rgba 0.6354 0.2779 0.182 0.8>
    R1 Cobalt:LinearColor=<rgba 0.2917 0.3048 1 1>
    R1 Crimson:LinearColor=<rgba 0.75 0.125 0.1419 1>
    R1 ForestGreen:LinearColor=<rgba 0.1081 0.8646 0.1463 1>
    R1 Grey:LinearColor=<rgba 0.3958 0.3958 0.3958 1>
    R1 Lime:LinearColor=<rgba 0.7812 1.5 0.0625 1>
    R1 Orange:LinearColor=<rgba 1.3 0.4534 0 1.2>
    R1 Pink:LinearColor=<rgba 1.5 0.5234 1.1449 1.2>
    R1 Purple:LinearColor=<rgba 0.6252 0.225 1.2 1>
    R1 Saffron:LinearColor=<rgba 1.2 0.92 0 1.1>
    R1 SkyBlue:LinearColor=<rgba 0.1693 0.9216 1.3 1.2>
    R1 TitaniumWhite:LinearColor=<rgba 1.5 1.5 1.5 1.2>
    Saffron=<rgba 0.4807 0.5104 0 1>
    SkyBlue=<rgba 0 0.3639 0.5469 1>
    TitaniumWhite=<rgba 0.6667 0.6667 0.6667 1>
    Use R1 Painted Overrides=true
   )
   &InitializeParticle(
    Color=&MakeCustomLinearColorFromBool(Bool=$User.bIsPainted, True Color=$Particles.Color, False Color=<rgba 0.7344 0.4025 0.2486 1>)
    Lifetime Max=1.15
    Lifetime Min=0.65
    Lifetime Mode=::NewEnumerator1
    Mass Mode=::NewEnumerator3
    Sprite Rotation Angle=&RandomRangeFloat(Minimum=-1.5, Maximum=1.5)
    Sprite Rotation Mode=::NewEnumerator1
    Sprite Size Mode=::NewEnumerator1
    Uniform Sprite Size Max=25
    Uniform Sprite Size Min=20
    UsePositionOffset=true
    Write Material Random=true
   )
   &ShapeLocation(MaskSpawnGroup=false, Override Seed=false, Sphere Radius=20, UseOwnerScale=true)
   &PartitionParticles(Number Of Partitions=$Emitter.PortCount)
   &SkeletalMeshLocation(
    Bone / Socket Index=$Particles.PartitionParticles.Partition
    Bone / Socket Sampling Mode=::NewEnumerator5
    MaskSpawnGroup=false
    Mesh Sampling Type=::NewEnumerator6
    Sampled Position Offset=&AddVectorToPosition(Position=<-40 0 0>, Vector=$User.DTCarOffset)
    Skeletal Mesh=$System.SkeletalMesh
   )
   &AddVelocity(Velocity=&MakeCustomVectorFromBool(
     Bool=$User.bIsFrontendPreview
     True Vector=&RandomRangeVector(Minimum=<-800 -100 0>, Maximum=<-800 100 0>)
     False Vector=&RandomRangeVector(Minimum=<-800 -100 0>, Maximum=<-800 100 0>)
    ))
  ]
  particleUpdateScript [
   &ParticleState()
   &ScaleSpriteSize(Uniform Curve Sprite Scale=DataInterface() {propertyValues {Curve {keys [{arriveTangent=0.8448, interpMode=RCIM_Cubic, leaveTangent=0.8448, tangentMode=RCTM_User, time=0, value=3}, <at 0.8 10 RCIM_Cubic>]}}})
   &ScaleColor(
    Scale Alpha=&FloatFromCurve(FloatCurve=DataInterface() {propertyValues {Curve {keys [<at -0.0046 1.0113 RCIM_Cubic>,<at 1.0011 0.002 RCIM_Cubic>]}}}, Scale Curve=1.1)
    Scale RGB=&VectorFromFloat(Value=1.5)
    ScaleA=true
    ScaleRGB=true
   )
   &InheritSourceMovement(
    Applied Position Scale=&Multiply_VectorByFloat(
     Vector=<1 1 1>
     Float=&FloatFromCurve(FloatCurve=DataInterface() {propertyValues {Curve {keys [
          {interpMode=RCIM_Cubic, leaveTangent=-0.0051, tangentMode=RCTM_Break, time=0, value=1}
          {arriveTangent=-0.247, interpMode=RCIM_Cubic, tangentMode=RCTM_Break, time=1, value=0.9}
         ]}}})
    )
    ApplyForce=false
    ApplyPosition=true
    ApplyPositionFalloff=false
   )
   &SolveForcesAndVelocity(Clamp Velocity=false, Limit Acceleration=false)
   &SubUVAnimation(
    "Sprite Renderer "=DataInterface() {propertyValues {SpriteRenderer=&self:MainTrail_NewTemplate001_0.NiagaraSpriteRendererProperties_0}}
    SubUV Animation Mode:ENiagara_SubUVLookupModeV2=::NewEnumerator1
    SubUV Curve=DataInterface() {propertyValues {Curve {keys [
        {arriveTangent=1.6219, interpMode=RCIM_Cubic, leaveTangent=1.6219, tangentMode=RCTM_User, time=0, value=0.5}
        {arriveTangent=-0.2228, interpMode=RCIM_Cubic, leaveTangent=-0.2228, tangentMode=RCTM_User, time=0.2, value=0.58}
        {arriveTangent=0.2086, interpMode=RCIM_Cubic, leaveTangent=0.2086, tangentMode=RCTM_User, time=1, value=1}
       ]}}}
    UseEndFrame=false
    UseStartFrame=false
   )
   /SetVariables(Particles.SparksColor:LinearColor=&MultiplyLinearColor(
     Linear Color=&MakeCustomLinearColorFromBool(Bool=$User.bIsPainted, True Color=$Particles.R1 Color, False Color=<rgba 2 1.0961 0.6771 1>)
     Scale Factor=&MakeVector4FromVectorAndFloat(
      "Vector (RGB)"=&VectorFromFloat(Value=5)
      "Float (Alpha)"=&FloatFromCurve(FloatCurve=DataInterface() {propertyValues {Curve {keys [<at -0.0046 1.0113 RCIM_Cubic>, {arriveTangent=-1.9677, interpMode=RCIM_Cubic, leaveTangent=-1.9677, tangentMode=RCTM_User, time=1, value=0}]}}})
     )
    )) {isSetParameters=true}
  ]
  rendererClasses [&NiagaraSpriteRendererProperties]
  renderers [
   {rendererClass=&NiagaraSpriteRendererProperties}
   {rendererClass=&NiagaraSpriteRendererProperties, rendererIndex=1}
   {rendererClass=&NiagaraSpriteRendererProperties, rendererIndex=2}
  ]
 }
 Rocks_NewTemplate001 {
  emitterUpdateScript [
   &EmitterState(Life Cycle Mode=::NewEnumerator1, Loop Duration=2, UseLoopDelay=false)
   &VC_FilteredSocketCount()
   &SpawnRate(
    SpawnRate=&MakeCustomFloatFromBool(Bool=$User.bIsFrontendPreview, True Float=12, False Float=12)
    Use Spawn Probability=false
   )
  ]
  particleSpawnScript [
   &VC_PaintedVariantsOverride(
    Black=<rgba 0.012 0.012 0.012 8>
    BurntSienna=<rgba 0.0365 0.0113 0.0045 8>
    Cobalt=<rgba 0.0134 0.0171 0.1979 8>
    Crimson=<rgba 0.0833 0 0.003 8>
    ForestGreen=<rgba 0.0057 0.099 0 8>
    Grey=<rgba 0.0365 0.0365 0.0365 8>
    Lime=<rgba 0.0743 0.1875 0.001 8>
    Orange=<rgba 0.1927 0.0694 0.0007 8>
    Pink=<rgba 0.2031 0.0571 0.1982 8>
    Purple=<rgba 0.0396 0 0.0938 8>
    Saffron=<rgba 0.1437 0.151 0 8>
    SkyBlue=<rgba 0 0.097 0.1458 8>
    TitaniumWhite=<rgba 0.1458 0.1458 0.1458 8>
   )
   &InitializeParticle(
    Color=&MakeCustomLinearColorFromBool(Bool=$User.bIsPainted, True Color=$Particles.Color, False Color=<rgba 0.349 0.1916 0.1218 5>)
    Lifetime Max=1
    Lifetime Min=0.8
    Lifetime Mode=::NewEnumerator1
    Mass Mode=::NewEnumerator3
    Sprite Rotation Angle=&RandomRangeFloat(Minimum=-1, Maximum=1)
    Sprite Rotation Mode=::NewEnumerator1
    Sprite Size Mode=::NewEnumerator1
    Uniform Sprite Size Max=30
    Uniform Sprite Size Min=20
    UsePositionOffset=false
    Write Material Random=true
   )
   &ShapeLocation(MaskSpawnGroup=false, Override Seed=false, Sphere Radius=40, UseOwnerScale=true)
   &PartitionParticles(Number Of Partitions=$Emitter.PortCount)
   &SkeletalMeshLocation(
    Bone / Socket Index=$Particles.PartitionParticles.Partition
    Bone / Socket Sampling Mode=::NewEnumerator5
    MaskSpawnGroup=false
    Mesh Sampling Type=::NewEnumerator6
    Sampled Position Offset=&AddVectorToPosition(Position=<-40 0 0>, Vector=$User.DTCarOffset)
    Skeletal Mesh=$System.SkeletalMesh
   )
   &AddVelocity(Velocity=&MakeCustomVectorFromBool(
     Bool=$User.bIsFrontendPreview
     True Vector=&RandomRangeVector(Minimum=<-850 -45 200>, Maximum=<-850 45 600>)
     False Vector=&RandomRangeVector(Minimum=<-850 -45 75>, Maximum=<-850 45 455>)
    ))
   &AccelerationForce(Acceleration=<0 0 -51.2>)
   &ApplyInitialForces()
  ]
  particleUpdateScript [
   &ParticleState()
   &SpriteRotationRate(Rotation Rate=6)
   &InheritSourceMovement(
    Applied Position Scale=&Multiply_VectorByFloat(
     Vector=<1 1 1>
     Float=&FloatFromCurve(FloatCurve=DataInterface() {propertyValues {Curve {keys [
          {interpMode=RCIM_Cubic, leaveTangent=-0.0051, tangentMode=RCTM_Break, time=0, value=1}
          {arriveTangent=-0.247, interpMode=RCIM_Cubic, tangentMode=RCTM_Break, time=1, value=0.9}
         ]}}})
    )
    ApplyForce=false
    ApplyPosition=true
    ApplyPositionFalloff=false
   )
   &ScaleColor(
    Scale Alpha=&FloatFromCurve(FloatCurve=DataInterface() {propertyValues {Curve {keys [<at -0.0046 1.0113 RCIM_Cubic>,<at 1.0011 0.002 RCIM_Cubic>]}}})
    ScaleA=true
    ScaleRGB=true
   )
   &GravityForce()
   &SolveForcesAndVelocity(Clamp Velocity=false, Limit Acceleration=false)
   &SubUVAnimation(
    End Frame Range Override=16
    Playback Scale Curve=DataInterface() {propertyValues {Curve {keys [
        {interpMode=RCIM_Cubic, tangentMode=RCTM_User, time=0.25, value=1}
        {arriveTangent=-1.2656, interpMode=RCIM_Cubic, leaveTangent=-1.2656, tangentMode=RCTM_User, time=1, value=0}
       ]}}}
    "Sprite Renderer "=DataInterface() {propertyValues {SpriteRenderer=&self:Rocks_NewTemplate001_0.NiagaraSpriteRendererProperties_0}}
    SubUV Animation Mode:ENiagara_SubUVLookupModeV2=::NewEnumerator3
    UseEndFrame=true
    UsePlayRateScale=false
    UseStartFrame=true
   )
   &DynamicMaterialParameters(
    Index 0 Param 1=&RandomRangeFloat(Minimum=0, Maximum=1)
    Module.Index0Param0Write=true
    Module.Index0Param1Write=true
    Module.Index0Param2Write=true
    Module.Index0Param3Write=true
   )
  ]
  rendererClasses [&NiagaraSpriteRendererProperties]
  renderers [{rendererClass=&NiagaraSpriteRendererProperties}]
 }
 StartFiller_NewTemplate001 {
  emitterUpdateScript [
   &EmitterState()
   &VC_FilteredSocketCount()
   &SpawnRate(
    SpawnRate=&MakeCustomFloatFromBool(Bool=$User.bIsFrontendPreview, True Float=30, False Float=30)
    Use Spawn Probability=false
   )
  ]
  particleSpawnScript [
   &VC_PaintedVariantsOverride(
    Black=<rgba 0.02 0.02 0.02 1.2>
    BurntSienna=<rgba 0.0938 0.0289 0.0116 1>
    Cobalt=<rgba 0 0.0185 1 1>
    ForestGreen=<rgba 0 0.3 0.0152 1.2>
    Grey=<rgba 0.125 0.125 0.125 1>
    Lime=<rgba 0.3136 0.724 0 1>
    Orange=<rgba 0.4427 0.1544 0 1>
    Pink=<rgba 0.7156 0.1872 0.7188 1.2>
    Purple=<rgba 0.1261 0 0.3073 1>
    R1 Black:LinearColor=<rgba 0.4219 0.4219 0.4219 0.5>
    R1 BurntSienna:LinearColor=<rgba 0.6354 0.2779 0.182 0.8>
    R1 Cobalt:LinearColor=<rgba 0.2917 0.3048 1 1>
    R1 Crimson:LinearColor=<rgba 0.75 0.125 0.1419 1>
    R1 ForestGreen:LinearColor=<rgba 0.0951 0.7604 0.1287 1>
    R1 Grey:LinearColor=<rgba 0.276 0.276 0.276 1>
    R1 Lime:LinearColor=<rgba 0.625 1.2 0.05 1>
    R1 Orange:LinearColor=<rgba 0.9635 0.3361 0 1>
    R1 Pink:LinearColor=<rgba 1 0.349 0.7633 1>
    R1 Purple:LinearColor=<rgba 0.4206 0.1514 0.8073 1>
    R1 SkyBlue:LinearColor=<rgba 0.1139 0.6203 0.875 1>
    R1 TitaniumWhite:LinearColor=<rgba 0.9062 0.9062 0.9062 1>
    Saffron=<rgba 0.4807 0.5104 0 1>
    SkyBlue=<rgba 0 0.3639 0.5469 1>
    TitaniumWhite=<rgba 0.6667 0.6667 0.6667 1>
    Use R1 Painted Overrides=true
   )
   &InitializeParticle(
    Color=&MakeCustomLinearColorFromBool(Bool=$User.bIsPainted, True Color=$Particles.Color, False Color=<rgba 0.7344 0.4025 0.2486 1>)
    Lifetime Max=0.15
    Lifetime Min=0.1
    Lifetime Mode=::NewEnumerator1
    Mass Mode=::NewEnumerator3
    Position Offset=<-2 0 0>
    Sprite Rotation Angle=&RandomRangeFloat(Minimum=-1, Maximum=1)
    Sprite Rotation Mode=::NewEnumerator1
    Sprite Size Mode=::NewEnumerator1
    Uniform Sprite Size Max=30
    Uniform Sprite Size Min=25
    UsePositionOffset=true
    Write Material Random=true
   )
   &ShapeLocation(MaskSpawnGroup=false, Override Seed=false, Sphere Radius=8, UseOwnerScale=true)
   &PartitionParticles(Number Of Partitions=$Emitter.PortCount)
   &SkeletalMeshLocation(
    Bone / Socket Index=$Particles.PartitionParticles.Partition
    Bone / Socket Sampling Mode=::NewEnumerator5
    MaskSpawnGroup=false
    Mesh Sampling Type=::NewEnumerator6
    Sampled Position Offset=&AddVectorToPosition(Position=<-40 0 0>, Vector=$User.DTCarOffset)
    Skeletal Mesh=$System.SkeletalMesh
   )
   &AddVelocity(Velocity=&MakeCustomVectorFromBool(Bool=$User.bIsFrontendPreview, True Vector=<-800 0 50>, False Vector=<-800 0 50>))
  ]
  particleUpdateScript [
   &ParticleState()
   &ScaleSpriteSize(Uniform Curve Sprite Scale=DataInterface() {propertyValues {Curve {keys [{tangentMode=RCTM_User, time=0, value=2}, {tangentMode=RCTM_User, time=1, value=5}]}}})
   &InheritSourceMovement(
    Applied Position Scale=&Multiply_VectorByFloat(
     Vector=<1 1 1>
     Float=&FloatFromCurve(FloatCurve=DataInterface() {propertyValues {Curve {keys [
          {interpMode=RCIM_Cubic, leaveTangent=-0.0051, tangentMode=RCTM_Break, time=0, value=1}
          {arriveTangent=-0.247, interpMode=RCIM_Cubic, tangentMode=RCTM_Break, time=1, value=0.9}
         ]}}})
    )
    ApplyForce=false
    ApplyPosition=true
    ApplyPositionFalloff=false
   )
   &SolveForcesAndVelocity(Clamp Velocity=false, Limit Acceleration=false)
   &ScaleColor(
    Scale Alpha=&FloatFromCurve(FloatCurve=DataInterface() {propertyValues {Curve {keys [<at -0.0046 1.0113 RCIM_Cubic>,<at 1.0011 0.002 RCIM_Cubic>]}}}, Scale Curve=1.1)
    Scale RGB=&VectorFromFloat(Value=1.5)
    ScaleA=true
    ScaleRGB=true
   )
   &SubUVAnimation(
    "Sprite Renderer "=DataInterface() {propertyValues {SpriteRenderer=&self:StartFiller_NewTemplate001_0.NiagaraSpriteRendererProperties_0}}
    SubUV Animation Mode:ENiagara_SubUVLookupModeV2=::NewEnumerator1
    SubUV Curve=DataInterface() {propertyValues {Curve {keys [
        {arriveTangent=1.6219, interpMode=RCIM_Cubic, leaveTangent=1.6219, tangentMode=RCTM_User, time=0, value=0.5}
        {arriveTangent=-0.2228, interpMode=RCIM_Cubic, leaveTangent=-0.2228, tangentMode=RCTM_User, time=0.2, value=0.58}
        {arriveTangent=0.2086, interpMode=RCIM_Cubic, leaveTangent=0.2086, tangentMode=RCTM_User, time=1, value=1}
       ]}}}
    UseEndFrame=false
    UseStartFrame=false
   )
   /SetVariables(Particles.SparksColor_1:LinearColor=&MultiplyLinearColor(
     Linear Color=&MakeCustomLinearColorFromBool(Bool=$User.bIsPainted, True Color=$Particles.R1 Color, False Color=<rgba 2 1.0961 0.6771 1>)
     Scale Factor=&MakeVector4FromVectorAndFloat(
      "Vector (RGB)"=&VectorFromFloat(Value=5)
      "Float (Alpha)"=&FloatFromCurve(FloatCurve=DataInterface() {propertyValues {Curve {keys [<at -0.0046 1.0113 RCIM_Cubic>, {arriveTangent=-1.9677, interpMode=RCIM_Cubic, leaveTangent=-1.9677, tangentMode=RCTM_User, time=1, value=0}]}}})
     )
    )) {isSetParameters=true}
  ]
  rendererClasses [&NiagaraSpriteRendererProperties]
  renderers [
   {rendererClass=&NiagaraSpriteRendererProperties}
   {rendererClass=&NiagaraSpriteRendererProperties, rendererIndex=1}
  ]
 }
 Twig_NewTemplate001 {
  emitterUpdateScript [
   &EmitterState(Life Cycle Mode=::NewEnumerator1, Loop Duration=2, UseLoopDelay=false)
   &VC_FilteredSocketCount()
   &SpawnRate(
    SpawnRate=&MakeCustomFloatFromBool(Bool=$User.bIsFrontendPreview, True Float=10, False Float=10)
    Use Spawn Probability=false
   )
  ]
  particleSpawnScript [
   &VC_PaintedVariantsOverride(
    Black=<rgba 0.02 0.02 0.02 20>
    BurntSienna=<rgba 0.04 0.0124 0.0049 20>
    Cobalt=<rgba 0.0212 0.0239 0.1562 20>
    Crimson=<rgba 0.06 0 0.0016 10>
    ForestGreen=<rgba 0 0.1094 0.0056 9>
    Grey=<rgba 0.06 0.06 0.06 10>
    Lime=<rgba 0.0999 0.1562 0.0366 10>
    Orange=<rgba 0.125 0.045 0.0005 8>
    Pink=<rgba 0.1719 0.0483 0.127 8>
    Purple=<rgba 0.0464 0 0.11 8>
    Saffron=<rgba 0.0761 0.08 0 8>
    SkyBlue=<rgba 0 0.1065 0.16 10>
    TitaniumWhite=<rgba 0.1667 0.1667 0.1667 5>
   )
   &InitializeParticle(
    Color Mode=::NewEnumerator0
    Lifetime Max=1
    Lifetime Min=0.8
    Lifetime Mode=::NewEnumerator1
    Mass Mode=::NewEnumerator3
    Sprite Rotation Angle=&RandomRangeFloat(Minimum=-1, Maximum=1)
    Sprite Rotation Mode=::NewEnumerator1
    Sprite Size Mode=::NewEnumerator1
    Uniform Sprite Size Max=&RandomRangeFloat(Minimum=60, Maximum=90)
    Uniform Sprite Size Min=&RandomRangeFloat(Minimum=35, Maximum=45)
    UsePositionOffset=false
    Write Material Random=true
   )
   &ShapeLocation(MaskSpawnGroup=false, Override Seed=false, Sphere Radius=40, UseOwnerScale=true)
   &Color(Color=&MakeCustomLinearColorFromBool(
     Bool=$User.bIsPainted
     True Color=&RandomRangeLinearColor(
      Minimum=&MultiplyLinearColor(Linear Color=$Particles.Color, Scale Factor=<0.8 0.8 0.8 0.8>)
      Maximum=$Particles.Color
     )
     False Color=&RandomRangeLinearColor(Minimum=<rgba 0.1875 0.1191 0.0605 7>, Maximum=<rgba 0.1562 0.102 0.0553 7>)
    ))
   &PartitionParticles(Number Of Partitions=$Emitter.PortCount)
   &AccelerationForce(Acceleration=<0 0 -51.2>)
   &ApplyInitialForces()
   &SkeletalMeshLocation(
    Bone / Socket Index=$Particles.PartitionParticles.Partition
    Bone / Socket Sampling Mode=::NewEnumerator5
    MaskSpawnGroup=false
    Mesh Sampling Type=::NewEnumerator6
    Sampled Position Offset=&AddVectorToPosition(Position=<-40 0 0>, Vector=$User.DTCarOffset)
    Skeletal Mesh=$System.SkeletalMesh
   )
   &AddVelocity(Velocity=&MakeCustomVectorFromBool(
     Bool=$User.bIsFrontendPreview
     True Vector=&RandomRangeVector(Minimum=<-850 -45 90>, Maximum=<-850 45 600>)
     False Vector=&RandomRangeVector(Minimum=<-850 -45 75>, Maximum=<-850 45 455>)
    ))
  ]
  particleUpdateScript [
   &ParticleState()
   &SpriteRotationRate(Rotation Rate=6)
   &ScaleColor(
    Scale Alpha=&FloatFromCurve(FloatCurve=DataInterface() {propertyValues {Curve {keys [<at -0.0046 1.0113 RCIM_Cubic>,<at 1.0011 0.002 RCIM_Cubic>]}}})
    ScaleA=true
    ScaleRGB=true
   )
   &GravityForce()
   &InheritSourceMovement(
    Applied Position Scale=&Multiply_VectorByFloat(
     Vector=<1 1 1>
     Float=&FloatFromCurve(FloatCurve=DataInterface() {propertyValues {Curve {keys [
          {interpMode=RCIM_Cubic, leaveTangent=-0.0051, tangentMode=RCTM_Break, time=0, value=1}
          {arriveTangent=-0.247, interpMode=RCIM_Cubic, tangentMode=RCTM_Break, time=1, value=0.9}
         ]}}})
    )
    ApplyForce=false
    ApplyPosition=true
    ApplyPositionFalloff=false
   )
   &SolveForcesAndVelocity(Clamp Velocity=false, Limit Acceleration=false)
   &SubUVAnimation(
    End Frame Range Override=16
    Playback Scale Curve=DataInterface() {propertyValues {Curve {keys [
        {interpMode=RCIM_Cubic, tangentMode=RCTM_User, time=0.25, value=1}
        {arriveTangent=-1.2656, interpMode=RCIM_Cubic, leaveTangent=-1.2656, tangentMode=RCTM_User, time=1, value=0}
       ]}}}
    "Sprite Renderer "=DataInterface() {propertyValues {SpriteRenderer=&self:Twig_NewTemplate001_0.NiagaraSpriteRendererProperties_0}}
    SubUV Animation Mode:ENiagara_SubUVLookupModeV2=::NewEnumerator3
    UseEndFrame=true
    UsePlayRateScale=false
    UseStartFrame=true
   )
  ]
  rendererClasses [&NiagaraSpriteRendererProperties]
  renderers [{rendererClass=&NiagaraSpriteRendererProperties}]
 }
}
summary {
 systemName=NS_Boost_SandStorm
 emitters [
  {emitterName=StartFiller_NewTemplate001, rendererClasses [&NiagaraSpriteRendererProperties]}
  {emitterName=MainTrail_NewTemplate001, rendererClasses [&NiagaraSpriteRendererProperties]}
  {emitterName=Twig_NewTemplate001, rendererClasses [&NiagaraSpriteRendererProperties]}
  {emitterName=Rocks_NewTemplate001, rendererClasses [&NiagaraSpriteRendererProperties]}
 ]
 userVariables [
  User.bIsFrontendPreview {type=NiagaraBool}
  User.bIsPainted {type=NiagaraBool}
  User.ColorVariantID {type=NiagaraInt32, defaultValue=1}
  User.DTCarOffset {type=Vector3f, description="", defaultValue {struct=Vector3f, value=<0 0 0>}}
  User.PaintedColor {type=LinearColor, description="", defaultValue {struct=LinearColor, value=<rgba 1 1 1 1>}}
  User.PaintedColor_2 {type=LinearColor, description="", defaultValue {struct=LinearColor, value=<rgba 1 1 1 1>}}
  User.PaintedScalar {type=NiagaraFloat, defaultValue=1, description=""}
  User.PortCount {type=NiagaraInt32, defaultValue=1}
 ]
}
